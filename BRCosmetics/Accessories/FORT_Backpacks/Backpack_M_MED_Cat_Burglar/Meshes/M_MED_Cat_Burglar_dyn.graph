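# Physics asset for the Cat Burglar backpack cosmetic, providing bodies and joints for its dangling dagger, chain and belts.
# A box body on the pack anchors a box dagger, two separate belt capsules and a three-capsule chain linked in series.
# Chain and dagger joints lock twist and one swing axis, so they swing in a single plane, while belt joints leave those axes unlocked.
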PhysicsAsset /BRCosmetics/Accessories/FORT_Backpacks/Backpack_M_MED_Cat_Burglar/Meshes/M_MED_Cat_Burglar_dyn
bodies {
 dyn_belt_01 [{shapeType=Capsule, length=4, radius=1, center=<3 0 0>, rotation=<rot 90 0 0>}]
 dyn_belt_02 [{shapeType=Capsule, length=4, radius=1, center=<2 0 0>, rotation=<rot 90 0 0>}]
 dyn_chain_01 [{shapeType=Capsule, length=3, radius=1, center=<2 0 0>, rotation=<rot 90 0 0>}]
 dyn_chain_02 [{shapeType=Capsule, length=3, radius=1, center=<1.5 0 0>, rotation=<rot 90 0 0>}]
 dyn_chain_03 [{shapeType=Capsule, length=3, radius=1, center=<1.5 0 0>, rotation=<rot 90 0 0>}]
 dyn_dagger [{shapeType=Box, extentX=38.6221, extentY=1.8262, extentZ=5.2561, center=<3 0 0>}]
 pack [{shapeType=Box, extentX=25.987, extentY=7.5, extentZ=17.1408, center=<-1.9348 -2.8345 -0.4148>, rotation=<rot 24.9999 0 0>}]
}
constraints [
 {
  bone1Name=dyn_chain_01
  bone2Name=pack
  swing1LimitDegrees=45
  swing1Motion=Locked
  swing2LimitDegrees=15
  twistLimitDegrees=5
  twistMotion=Locked
 }
 {
  bone1Name=dyn_chain_02
  bone2Name=dyn_chain_01
  swing1LimitDegrees=2
  swing1Motion=Locked
  swing2LimitDegrees=15
  twistLimitDegrees=2
  twistMotion=Locked
 }
 {
  bone1Name=dyn_dagger
  bone2Name=pack
  swing1LimitDegrees=45
  swing1Motion=Locked
  swing2LimitDegrees=5
  twistLimitDegrees=45
  twistMotion=Locked
 }
 {bone1Name=dyn_belt_01, bone2Name=pack, swing1LimitDegrees=30, swing2LimitDegrees=30, twistLimitDegrees=5}
 {bone1Name=dyn_belt_02, bone2Name=pack, swing1LimitDegrees=30, swing2LimitDegrees=30, twistLimitDegrees=5}
 {
  bone1Name=dyn_chain_03
  bone2Name=dyn_chain_02
  swing1LimitDegrees=2
  swing1Motion=Locked
  swing2LimitDegrees=15
  twistLimitDegrees=2
  twistMotion=Locked
 }
]
